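# Sound cue for the coal item landing when dropped, using the Fort_Loot_Collect sound class and weapon-drop attenuation.
# A single coal drop wave feeds a mixer that boosts its input, then a modulator at the root.
# A concurrency override caps simultaneous instances, stopping the oldest when the limit is exceeded.
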SoundCue /Game/Athena/Sounds/Items/Coal/Coal_Drop_Land_Cue
root=Modulator_5
nodes [
 {id=Mixer_0, children [WavePlayer_1], properties {inputVolume [2]}}
 {id=Modulator_5, children [Mixer_0]}
 {id=WavePlayer_1, properties {soundWaveAssetPtr=&/Weapon_Item_Coal_Drop_02}}
]
properties {
 attenuationSettings=&Fort_Weapon_Drop_Att
 bOverrideConcurrency=true
 maxDistance=3000
 soundClassObject=&Fort_Loot_Collect
 volumeMultiplier=0.398
 concurrencyOverrides {maxCount=2, resolutionRule=StopOldest}
}
