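# Sound cue for equipping the Coal item, classed as player foley.
# It plays a single equip wave through a modulator left at its usual settings.
# A boosted volume multiplier and the shared weapon reload attenuation shape its loudness and falloff.
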SoundCue /Game/Athena/Sounds/Items/Coal/Coal_Equip_Cue
root=Modulator_0
nodes [
 {id=Modulator_0, children [WavePlayer_1]}
 {id=WavePlayer_1, properties {soundWaveAssetPtr=&/Weapon_Item_Coal_Equip_01}}
]
properties {
 attenuationSettings=&Fort_Weapon_Reload_Att
 maxDistance=2250
 soundClassObject=&Fort_Player_Foley
 volumeMultiplier=2.5
}
