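# Looping in-air sound cue for the Coal item projectile in Athena, heard while it travels.
# A looping wave player feeds the coal in-air sound through a root Modulator node left at its default settings.
# The cue uses a dedicated coal in-air loop attenuation, the ranged projectile effects sound class, and a heavily boosted volume multiplier.
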SoundCue /Game/Athena/Sounds/Items/Coal/Coal_InAir_Loop_Cue
root=Modulator_0
nodes [
 {id=Modulator_0, children [WavePlayer_1]}
 {id=WavePlayer_1, properties {bLooping=true, soundWaveAssetPtr=&/Weapon_Item_Coal_InAir_Loop}}
]
properties {
 attenuationSettings=/Game/Sounds/Attenuation/Fort_Weapon_Projectile_Coal_InAirLoop_Att
 maxDistance=2097152
 soundClassObject=&Fort_WeaponRanged_Projectile_Effects
 volumeMultiplier=10
}
