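# Sound cue that plays when a coal item is picked up.
# A single wave player feeds a Modulator at default settings, adding slight random pitch and volume variation to one sample.
# It uses weapon pickup attenuation and the loot-collect sound class at slightly reduced volume.
# A concurrency override caps it at two simultaneous instances, stopping the oldest.
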SoundCue /Game/Athena/Sounds/Items/Coal/Coal_Pickup_Cue
root=Modulator_0
nodes [
 {id=Modulator_0, children [WavePlayer_2]}
 {id=WavePlayer_2, properties {soundWaveAssetPtr=&/Weapon_Item_Coal_Pick_Up_01}}
]
properties {
 attenuationSettings=&Fort_Weapon_Pickup_Att
 bOverrideConcurrency=true
 maxDistance=3500
 soundClassObject=&Fort_Loot_Collect
 volumeMultiplier=0.796
 concurrencyOverrides {maxCount=2, resolutionRule=StopOldest}
}
