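# Sound cue for throwing the Coal item in Athena.
# A Random node picks one of four equally weighted throw waves and passes it through a Modulator node left at its default settings.
# Playback is quieted by a reduced volume multiplier, routed to the gadget items sound class and spatialized with the gadget-placement attenuation.
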SoundCue /Game/Athena/Sounds/Items/Coal/Coal_Throw_Cue
root=Modulator_0
nodes [
 {id=Modulator_0, children [Random_0]}
 {id=Random_0, children [WavePlayer_0,WavePlayer_2,WavePlayer_3,WavePlayer_4], properties {weights [1,1,1,1]}}
 {id=WavePlayer_0, properties {soundWaveAssetPtr=&/Weapon_Item_Coal_Throw_01}}
 {id=WavePlayer_2, properties {soundWaveAssetPtr=&/Weapon_Item_Coal_Throw_02}}
 {id=WavePlayer_3, properties {soundWaveAssetPtr=&/Weapon_Item_Coal_Throw_03}}
 {id=WavePlayer_4, properties {soundWaveAssetPtr=&/Weapon_Item_Coal_Throw_04}}
]
properties {
 attenuationSettings=/Game/Sounds/Attenuation/Fort_GadgetPlace_Att
 maxDistance=2000
 soundClassObject=&Fort_Gadgets_Items
 volumeMultiplier=0.25
}
